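# Sound cue for the fourth main hero "Bank" sound in Section 03 of the PistolCanary event.
# A single wave player is the root and plays the matching Bank 04 wave, with no mixing, randomization or modulation nodes.
# It uses the shared main-hero whiz-by attenuation asset for long-range spatial falloff and routes to the Fort_Events sound class.
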SoundCue /PistolCanary_Event_Audio/Assets/Section03/MainHero/PistolCanary_Sec_03_Hero_Bank_04_Cue
root=WavePlayer_0
nodes [{id=WavePlayer_0, properties {soundWaveAssetPtr=/PistolCanary_Event_Audio/Assets/Section03/wav/PistolCanary_Sec_03_Hero_Bank_04}}]
properties {
 attenuationSettings=/PistolCanary_Audio/Assets/Settings/ATT_PistolCanary_MainHero_WhizBys
 maxDistance=53000
 soundClassObject=&Fort_Events
}
