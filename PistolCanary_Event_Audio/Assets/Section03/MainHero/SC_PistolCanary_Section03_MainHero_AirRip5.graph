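# Sound cue for the main hero's "AirRip" sound in Section 03 of the PistolCanary event audio.
# A random node picks one of five AirRip wave variations with equal weight.
# It plays at reduced volume under the Fort_Events sound class, using a shared main hero whiz-by attenuation asset with a long audible range.
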SoundCue /PistolCanary_Event_Audio/Assets/Section03/MainHero/SC_PistolCanary_Section03_MainHero_AirRip5
root=Random_0
nodes [
 {id=Random_0, children [WavePlayer_0,WavePlayer_1,WavePlayer_2,WavePlayer_3,WavePlayer_5], properties {weights [1,1,1,1,1]}}
 {id=WavePlayer_0, properties {soundWaveAssetPtr=/PistolCanary_Event_Audio/Assets/Section03/wav/PistolCanary_Seq_03_SM_AirRip_01}}
 {id=WavePlayer_1, properties {soundWaveAssetPtr=/PistolCanary_Event_Audio/Assets/Section03/wav/PistolCanary_Seq_03_SM_AirRip_02}}
 {id=WavePlayer_2, properties {soundWaveAssetPtr=/PistolCanary_Event_Audio/Assets/Section03/wav/PistolCanary_Seq_03_SM_AirRip_03}}
 {id=WavePlayer_3, properties {soundWaveAssetPtr=/PistolCanary_Event_Audio/Assets/Section03/wav/PistolCanary_Seq_03_SM_AirRip_04}}
 {id=WavePlayer_5, properties {soundWaveAssetPtr=/PistolCanary_Event_Audio/Assets/Section03/wav/PistolCanary_Seq_03_SM_AirRip_05}}
]
properties {
 attenuationSettings=/PistolCanary_Audio/Assets/Settings/ATT_PistolCanary_MainHero_WhizBys
 maxDistance=53000
 soundClassObject=&Fort_Events
 volumeMultiplier=0.6
}
