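# Sound cue for the camera approach moment in Section 03 of the PistolCanary event, in the MainHero audio set.
# A single wave player feeds the Section 03 camera approach wave straight to the output, with no randomization or mixing.
# It uses a close-range main hero attenuation asset with a long max distance, routes to the Fortnite events sound class, and slightly boosts volume.
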SoundCue /PistolCanary_Event_Audio/Assets/Section03/MainHero/SC_PistolCanary_Section03_MainHero_CameraApproach
root=WavePlayer_4
nodes [{id=WavePlayer_4, properties {soundWaveAssetPtr=/PistolCanary_Event_Audio/Assets/Section03/wav/PistolCanary_Sec03_SM_Camera_Approach}}]
properties {
 attenuationSettings=/PistolCanary_Audio/Assets/Settings/ATT_PistolCanary_MainHero_Close
 maxDistance=23000
 soundClassObject=&Fort_Events
 volumeMultiplier=1.2
}
