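# Sound cue for a Section 03 main hero laser shot (Spire 01) in the PistolCanary event audio.
# A single wave player plays one hero laser shoot wave directly, with no randomization or modulation.
# It uses the main hero weapon attenuation asset with a very long max distance, routes to the Fort_Events sound class, and plays at reduced volume.
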SoundCue /PistolCanary_Event_Audio/Assets/Section03/MainHero/SC_PistolCanary_Section03_MainHero_Laser_Spire01
root=WavePlayer_1
nodes [{id=WavePlayer_1, properties {soundWaveAssetPtr=/PistolCanary_Event_Audio/Assets/Section03/wav/PistolCanary_Sec_03_Hero_Laser_Shoot_01}}]
properties {
 attenuationSettings=/PistolCanary_Audio/Assets/Settings/ATT_PistolCanary_MainHero_Weapon
 maxDistance=151000
 soundClassObject=&Fort_Events
 volumeMultiplier=0.6
}
